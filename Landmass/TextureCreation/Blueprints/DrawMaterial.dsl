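; DrawMaterial is a texture-creation Blueprint in the Landmass content that renders a material into a render target.
; Its custom Draw event fetches the RT variable, clears it, then draws the Material variable into it.
; The construction script is empty, and the file shows nothing that calls Draw automatically.
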(event Custom|Draw
  (bind _rt (Variables|Default|GetRT))
  (Rendering|ClearRenderTarget2D _rt)
  (Rendering|DrawMaterialtoRenderTarget _rt (Variables|Default|GetMaterial)))

(fn ConstructionScript)
